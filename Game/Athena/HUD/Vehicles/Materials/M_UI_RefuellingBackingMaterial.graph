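# Additive UI material for the backing of the vehicle refuelling display in the HUD.
# A vertical panner's green channel is scaled and fed through an inverted sine to make scrolling horizontal bands.
# The bands are multiplied by the smooth output of a radial gradient function, shaping them radially.
# A teal Color vector parameter is exposed; its connection to the material outputs is not shown.
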Material /Game/Athena/HUD/Vehicles/Materials/M_UI_RefuellingBackingMaterial
graph {
 properties {
  bAutomaticallySetUsageInEditor=false
  blendMode=BLEND_Additive
  materialDomain=MD_UI
  pixelDepthOffsetMode=PDOM_Legacy
  usedShadingModels=Default Lit
 }
 nodes [
  Constant() {r=-0.128}
  Constant() {r=1.5682}
  n1 = Multiply(A=ComponentMask(None=Panner() {speedY=0.1}) {g=true}) {constB=15}
  Multiply(
   A=OneMinus(None=Sine(None=@n1))
   B=MaterialFunctionCall() {materialFunction=&MF_UI_RadialGradient}.Smooth
  )
  VectorParameter() {parameterName=Color, defaultValue=<rgba 0.0437 0.3325 0.2918 1>}
 ]
}
