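# M_UI_RefuellingShineMaterial is an additive UI material for the vehicle HUD, named for a refuelling shine effect.
# It contains a DiamondGradient function with a constant falloff, which gives a procedural diamond-shaped glow mask without a texture.
# A Color vector parameter, light blue by default, provides an adjustable tint.
# The graph has no animation nodes, and its links to the material outputs are not shown.
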Material /Game/Athena/HUD/Vehicles/Materials/M_UI_RefuellingShineMaterial
graph {
 properties {
  bAutomaticallySetUsageInEditor=false
  blendMode=BLEND_Additive
  materialDomain=MD_UI
  pixelDepthOffsetMode=PDOM_Legacy
  usedShadingModels=Default Lit
 }
 nodes [
  MaterialFunctionCall!(Falloff=Constant() {r=3}) {materialFunction=/Engine/Functions/Engine_MaterialFunctions01/Gradient/DiamondGradient}
  VectorParameter() {parameterName=Color, defaultValue=<rgba 0.1413 0.7835 1 1>}
 ]
}
