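# Vehicle HUD fuel gauge widget derived from FortVehicleFuelGauge, with an E-to-F material meter.
# Vehicle updates swap the fuel icon for a battery on electric-tagged vehicles, and an infinity label shows for infinite fuel.
# Refueling start plays a shine, ring and scale-up animation, tints the icon glow teal and enables meter material glow.
# The StartScaleLoop sequencer event runs an endless pulse, which refueling finish stops while resetting scale and glow.
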WidgetBlueprint /GreaseClipDunmurryUI/Widgets/WBP_GreaseClipDunmurry_FuelGauge
animations [
 Anim_ReFuelling {
  binding_count=3
  length_seconds=0.5013
  object=&self:Anim_ReFuelling
  sequence {
   bindings [
    RefuellingShine {
     id=#1
     possessed_class=Image
     tracks [
      {
       type=MovieSceneColorTrack
       display_name=Color and Opacity
       sections [{
         type=MovieSceneColorSection
         end_frame=6
         start_frame=-1
         channels [
          Color.R {type=MovieSceneScriptingFloatChannel, const [1,1]}
          Color.G {type=MovieSceneScriptingFloatChannel, const [1,1]}
          Color.B {type=MovieSceneScriptingFloatChannel, const [1,1]}
          Color.A {type=MovieSceneScriptingFloatChannel, keys [[-0.025,0],[0,0],[2.123,0.5],[4.0487,0]]}
         ]
        }]
      }
      {
       type=MovieScene2DTransformTrack
       sections [{
         type=MovieScene2DTransformSection
         end_frame=4
         start_frame=-1
         channels [
          Scale.X {type=MovieSceneScriptingFloatChannel, keys [[-0.025,1],[2.0043,4],[4,1]]}
          Scale.Y {type=MovieSceneScriptingFloatChannel, keys [[-0.025,1],[2.0043,4],[4,1]]}
         ]
        }]
      }
     ]
    }
    RefuellingCircle {
     id=#2
     possessed_class=Image
     tracks [{
       type=MovieSceneWidgetMaterialTrack
       sections [{
         type=MovieSceneParameterSection
         end_frame=10
         start_frame=0
         channels [
          Opacity {type=MovieSceneScriptingFloatChannel, keys [[0,0.0695],[2,1],[3.3857,0.5025],[7.0263,0]]}
          Scale {type=MovieSceneScriptingFloatChannel, keys [[0,0.5605],[4,0.8476],[7,0.9476]]}
          OutlineThickness {type=MovieSceneScriptingFloatChannel, keys [[0,0.1889],[1.972,0.1652],[3.4,0.107],[7.0147,0.028]]}
         ]
        }]
      }]
    }
    Overlay_FuelContainer {
     id=#3
     possessed_class=Overlay
     tracks [{
       type=MovieScene2DTransformTrack
       sections [{
         type=MovieScene2DTransformSection
         end_frame=10
         start_frame=0
         channels [
          Scale.X {type=MovieSceneScriptingFloatChannel, keys [[0,1],[4,1.1]]}
          Scale.Y {type=MovieSceneScriptingFloatChannel, keys [[0,1],[4,1.1]]}
         ]
        }]
      }]
    }
   ]
   tracks [{
     type=MovieSceneEventTrack
     sections [{
       type=MovieSceneEventTriggerSection
       channels [None {type=MovieSceneScriptingEventChannel, const ["<Struct 'MovieSceneEvent' (0x46887252580) {}>", 1]}]
      }]
    }]
  }
 }
 Anim_RefuellingScaleLoop {
  binding_count=1
  length_seconds=0.5
  object=&self:Anim_RefuellingScaleLoop
  sequence {bindings [Overlay_FuelContainer {
     id=#4
     possessed_class=Overlay
     tracks [{
       type=MovieScene2DTransformTrack
       sections [{
         type=MovieScene2DTransformSection
         end_frame=10
         start_frame=0
         channels [
          Scale.X {type=MovieSceneScriptingFloatChannel, keys [[0,1.1],[5,1.05],[10,1.1]]}
          Scale.Y {type=MovieSceneScriptingFloatChannel, keys [[0,1.1],[5,1.05],[10,1.1]]}
         ]
        }]
      }]
    }]}
 }
 Anim_ReFuelling_Reset {
  binding_count=2
  length_seconds=0.5
  object=&self:Anim_ReFuelling_Reset
  sequence {bindings [
    RefuellingCircle {
     id=#2
     possessed_class=Image
     tracks [{
       type=MovieSceneWidgetMaterialTrack
       sections [{
         type=MovieSceneParameterSection
         end_frame=10
         start_frame=0
         channels [
          Scale {type=MovieSceneScriptingFloatChannel, const [0.7,1]}
          OutlineThickness {type=MovieSceneScriptingFloatChannel, const [0.04,1]}
         ]
        }]
      }]
    }
    Overlay_FuelContainer {
     id=#3
     possessed_class=Overlay
     tracks [{
       type=MovieScene2DTransformTrack
       sections [{
         type=MovieScene2DTransformSection
         end_frame=10
         start_frame=0
         channels [
          Scale.X {type=MovieSceneScriptingFloatChannel, keys [[0,1.1],[4,1]]}
          Scale.Y {type=MovieSceneScriptingFloatChannel, keys [[0,1.1],[4,1]]}
         ]
        }]
      }]
    }
   ]}
 }
]
graphs [
 EventGraph {nodes [
   Event Construct()
   > n1 = GetDynamicMaterial(self=Get Image_Meter() -> (Image_Meter:Image Object Reference))
   > Set ImageSecondaryMeterMaterial(ImageSecondaryMeterMaterial:Material Instance Dynamic Object Reference=@n1) -> (Output_Get:Material Instance Dynamic Object Reference)
   n2 = Event OnVehicleUpdated!() -> (OutputDelegate, then, NewVehicle:Fort Athena Vehicle Object Reference) {is=title K2Node_Event}
   > n3 = Branch(Condition=HasMatchingGameplayTag!(self:Gameplay Tag Asset Interface Interface=@n2.NewVehicle, TagToCheck:Gameplay Tag Structure=(TagName="Vehicle.Fuel.Electric")) -> (ReturnValue:bool))
   > SetVisibility(self=Get Image_BatteryIcon() -> (Image_BatteryIcon:Image Object Reference), InVisibility=SelfHitTestInvisible)
   > SetVisibility(self=Get Image_MeterIcon() -> (Image_MeterIcon:Image Object Reference), InVisibility=Collapsed)
   SetVisibility(execute=@n3.else, self=Get Image_BatteryIcon(), InVisibility=Collapsed)
   > SetVisibility(self=Get Image_MeterIcon(), InVisibility=SelfHitTestInvisible)
   n4 = Event ShowInfiniteFuel!() -> (OutputDelegate, then, bHasInfiniteFuel:bool) {is=title K2Node_Event}
   > SetVisibility(
    self=Get Text_FuelInfinite() -> (Text_FuelInfinite:Common Text Object Reference)
    InVisibility=Select(Option 0=Collapsed, Option 1=HitTestInvisible, Index=@n4.bHasInfiniteFuel)
   )
   Event OnRefuelingCosmeticsStarted() {is=title K2Node_Event}
   > Parent: OnRefuelingCosmeticsStarted() {is=title K2Node_CallParentFunction}
   > PlayAnimationForward(InAnimation=Get Anim_ReFuelling() -> (Anim_ReFuelling:Widget Animation Object Reference))
   > SetVisibility(
    self=[Get RefuellingBacking() -> (RefuellingBacking:Image Object Reference), Get RefuellingCircle() -> (RefuellingCircle:Image Object Reference)]
    InVisibility=HitTestInvisible
   )
   > SetBrushTintColor!(
    execute
    self:Image Object Reference=Get Image_MeterIconGlow() -> (Image_MeterIconGlow:Image Object Reference)
    TintColor:Slate Color Structure="To SlateColor (Color)"!(Value:Linear Color Structure=(R=0.0467,G=0.5841,B=0.5089,A=1)) -> (ReturnValue:Slate Color Structure)
   ) -> (then)
   n5 = Get ImageSecondaryMeterMaterial() -> (ImageSecondaryMeterMaterial:Material Instance Dynamic Object Reference)
   > SetVectorParameterValue(self=@n5, ParameterName=FillLineColor, Value=(R=0.0471,G=0.5843,B=0.5098,A=1))
   > SetScalarParameterValue(self=@n5, ParameterName=EnableGlow, Value=1)
   n6 = Event OnRefuelingCosmeticsFinished!() -> (OutputDelegate, then, bIsFuelLow:bool) {is=title K2Node_Event}
   > Parent: OnRefuelingCosmeticsFinished!(execute, bIsFuelLow:bool=@n6.bIsFuelLow) -> (then) {is=title K2Node_CallParentFunction}
   > PlayAnimationForward(InAnimation=Get Anim_ReFuelling_Reset() -> (Anim_ReFuelling_Reset:Widget Animation Object Reference))
   > SetVisibility(self=[Get RefuellingBacking(), Get RefuellingCircle()], InVisibility=Collapsed)
   > SetBrushTintColor(self=Get Image_MeterIconGlow(), TintColor="To SlateColor (Color)"(Value=(R=1,G=0.1095,B=0,A=1)))
   > StopAnimation(InAnimation=Get Anim_RefuellingScaleLoop() -> (Anim_RefuellingScaleLoop:Widget Animation Object Reference))
   > SetScalarParameterValue(self=Get ImageSecondaryMeterMaterial(), ParameterName=EnableGlow, Value=0)
  ]}
 Sequencer Events {nodes [
   StartScaleLoop() {is=title K2Node_CustomEvent}
   > PlayAnimation(InAnimation=Get Anim_RefuellingScaleLoop(), NumLoopsToPlay=0, PlaybackSpeed=0.5, bRestoreState=true)
  ]}
]
variables []
widgets {
 info {inheritedWidgetCount=3, parentClass=/Script/FortniteUI.FortVehicleFuelGauge, rootWidgetClass=&Overlay, widgetCount=16}
 widgets [
  Overlay_FuelContainer {widgetClassPath=&Overlay, widget=Overlay_FuelContainer}
  RefuellingBacking {
   widgetClassPath=&Image
   bIsVariable=true
   parent=Overlay_FuelContainer
   widget=RefuellingBacking
   properties {
    renderOpacity=0.508
    visibility=Collapsed
    brush {
     bIsDynamicallyLoaded=false
     drawAs=Image
     imageType=FullColor
     mirroring=NoMirror
     resourceName=None
     resourceObject=/Game/Athena/HUD/Vehicles/Materials/M_UI_RefuellingBackingMaterial
     tiling=NoTile
     imageSize=<62.8129 113.4771>
     margin=<ltrb 0 0 0 0>
     outlineSettings {
      bUseBrushTransparency=false
      roundingType=HalfHeightRadius
      width=0
      color {colorUseRule=UseColor_Specified, specifiedColor=<rgba 0 0 0 0>}
      cornerRadii=<0 0 0 0>
     }
     tintColor {colorUseRule=UseColor_Specified, specifiedColor=<rgba 1 1 1 1>}
     uVRegion {bIsValid=false, max=<0 0>, min=<0 0>}
    }
    colorAndOpacity=<rgba 1 1 1 0.4>
    renderTransform {angle=0, scale=<1.35 1.25>, shear=<0 0>, translation=<-24 0>}
   }
   slot {class=OverlaySlot, properties {horizontalAlignment=HAlign_Center, verticalAlignment=VAlign_Center}}
  }
  RefuellingCircle {
   widgetClassPath=&Image
   bIsVariable=true
   parent=Overlay_FuelContainer
   widget=RefuellingCircle
   properties {
    visibility=Collapsed
    brush {
     bIsDynamicallyLoaded=false
     drawAs=Image
     imageType=FullColor
     mirroring=NoMirror
     resourceName=None
     resourceObject=/Game/Athena/HUD/Vehicles/Materials/M_UI_ReFuellingStation_VFX
     tiling=NoTile
     imageSize=<200 200>
     margin=<ltrb 0 0 0 0>
     outlineSettings {
      bUseBrushTransparency=false
      roundingType=HalfHeightRadius
      width=0
      color {colorUseRule=UseColor_Specified, specifiedColor=<rgba 0 0 0 0>}
      cornerRadii=<0 0 0 0>
     }
     tintColor {colorUseRule=UseColor_Specified, specifiedColor=<rgba 1 1 1 1>}
     uVRegion {bIsValid=false, max=<0 0>, min=<0 0>}
    }
   }
   slot {class=OverlaySlot, properties {horizontalAlignment=HAlign_Center, verticalAlignment=VAlign_Center}}
  }
  RefuellingShine {
   widgetClassPath=&Image
   bIsVariable=true
   parent=Overlay_FuelContainer
   widget=RefuellingShine
   properties {
    brush {
     bIsDynamicallyLoaded=false
     drawAs=Image
     imageType=FullColor
     mirroring=NoMirror
     resourceName=None
     resourceObject=/Game/Athena/HUD/Vehicles/Materials/M_UI_RefuellingShineMaterial
     tiling=NoTile
     imageSize=<64 64>
     margin=<ltrb 0 0 0 0>
     outlineSettings {
      bUseBrushTransparency=false
      roundingType=HalfHeightRadius
      width=0
      color {colorUseRule=UseColor_Specified, specifiedColor=<rgba 0 0 0 0>}
      cornerRadii=<0 0 0 0>
     }
     tintColor {colorUseRule=UseColor_Specified, specifiedColor=<rgba 1 1 1 1>}
     uVRegion {bIsValid=false, max=<0 0>, min=<0 0>}
    }
    colorAndOpacity=<rgba 1 1 1 0>
    renderTransform {angle=0, scale=<1 1>, shear=<0 0>, translation=<-40 -1>}
   }
   slot {class=OverlaySlot, properties {horizontalAlignment=HAlign_Center, verticalAlignment=VAlign_Center}}
  }
  Overlay_ExtensionSlot {widgetClassPath=&Overlay, bIsVariable=true, parent=Overlay_FuelContainer, widget=Overlay_ExtensionSlot, slot {class=OverlaySlot}}
  Overlay_Meter {
   widgetClassPath=&Overlay
   bInherited=true
   parent=Overlay_FuelContainer
   widget=Overlay_Meter
   slot {class=OverlaySlot, properties {horizontalAlignment=HAlign_Center, verticalAlignment=VAlign_Center}}
  }
  Image_Meter {
   widgetClassPath=&Image
   bIsVariable=true
   parent=Overlay_Meter
   widget=Image_Meter
   properties {brush {
     bIsDynamicallyLoaded=false
     drawAs=Image
     imageType=NoImage
     mirroring=NoMirror
     resourceName=None
     resourceObject=/Game/Athena/HUD/Vehicles/Materials/M_UI_ValetDash_Fuel
     tiling=NoTile
     imageSize=<200 200>
     margin=<ltrb 0 0 0 0>
     outlineSettings {
      bUseBrushTransparency=false
      roundingType=HalfHeightRadius
      width=0
      color {colorUseRule=UseColor_Specified, specifiedColor=<rgba 0 0 0 0>}
      cornerRadii=<0 0 0 0>
     }
     tintColor {colorUseRule=UseColor_Specified, specifiedColor=<rgba 1 1 1 1>}
     uVRegion {bIsValid=false, max=<0 0>, min=<0 0>}
    }}
   slot {class=OverlaySlot}
  }
  Text_FuelMax {
   widgetClassPath=&CommonTextBlock
   bIsVariable=true
   parent=Overlay_Meter
   widget=Text_FuelMax
   properties {
    text=F
    font {
     bForceMonospaced=false
     bMaterialIsStencil=false
     fontMaterial=null
     fontObject=&BurbankSmall
     letterSpacing=0
     monospacedWidth=1
     size=15
     skewAmount=0
     typefaceFontName=Black
     outlineSettings {
      bApplyOutlineToDropShadows=false
      bMiteredCorners=false
      bSeparateFillAlpha=false
      outlineMaterial=null
      outlineSize=0
      outlineColor=<rgba 0 0 0 1>
     }
    }
    renderTransform {angle=0, scale=<1 1>, shear=<-12 0>, translation=<-1 15>}
    shadowOffset=<0 0>
   }
   slot {class=OverlaySlot, properties {horizontalAlignment=HAlign_Center}}
  }
  Text_FuelMin {
   widgetClassPath=&CommonTextBlock
   bIsVariable=true
   parent=Overlay_Meter
   widget=Text_FuelMin
   properties {
    text=E
    font {
     bForceMonospaced=false
     bMaterialIsStencil=false
     fontMaterial=null
     fontObject=&BurbankSmall
     letterSpacing=0
     monospacedWidth=1
     size=15
     skewAmount=0
     typefaceFontName=Black
     outlineSettings {
      bApplyOutlineToDropShadows=false
      bMiteredCorners=false
      bSeparateFillAlpha=false
      outlineMaterial=null
      outlineSize=0
      outlineColor=<rgba 0 0 0 1>
     }
    }
    renderTransform {angle=0, scale=<1 1>, shear=<-12 0>, translation=<-6 -15>}
    shadowOffset=<0 0>
   }
   slot {class=OverlaySlot, properties {horizontalAlignment=HAlign_Center, verticalAlignment=VAlign_Bottom}}
  }
  Overlay_FuelIcon {
   widgetClassPath=&Overlay
   bInherited=true
   parent=Overlay_FuelContainer
   widget=Overlay_FuelIcon
   properties {renderTransform {angle=0, scale=<1 1>, shear=<0 0>, translation=<-28 0>}}
   slot {class=OverlaySlot, properties {horizontalAlignment=HAlign_Center, verticalAlignment=VAlign_Center}}
  }
  VerticalBox_0 {
   widgetClassPath=&VerticalBox
   parent=Overlay_FuelIcon
   widget=VerticalBox_0
   slot {class=OverlaySlot, properties {horizontalAlignment=HAlign_Center, verticalAlignment=VAlign_Bottom}}
  }
  Overlay_0 {widgetClassPath=&Overlay, parent=VerticalBox_0, widget=Overlay_0, slot {class=VerticalBoxSlot}}
  Image_MeterIconGlow {
   widgetClassPath=&Image
   bInherited=true
   bIsVariable=true
   parent=Overlay_0
   widget=Image_MeterIconGlow
   properties {
    visibility=Collapsed
    brush {
     bIsDynamicallyLoaded=false
     drawAs=Image
     imageType=NoImage
     mirroring=NoMirror
     resourceName=None
     resourceObject=/Game/Athena/HUD/LTM/Wax/Materials/M_UI_TextGlow_Pulse_Add
     tiling=NoTile
     imageSize=<50 50>
     margin=<ltrb 0 0 0 0>
     outlineSettings {
      bUseBrushTransparency=false
      roundingType=HalfHeightRadius
      width=0
      color {colorUseRule=UseColor_Specified, specifiedColor=<rgba 0 0 0 0>}
      cornerRadii=<0 0 0 0>
     }
     tintColor {colorUseRule=UseColor_Specified, specifiedColor=<rgba 0.8906 0.1088 0 1>}
     uVRegion {bIsValid=false, max=<0 0>, min=<0 0>}
    }
   }
   slot {class=OverlaySlot, properties {horizontalAlignment=HAlign_Center, verticalAlignment=VAlign_Center}}
  }
  Image_MeterIcon {
   widgetClassPath=&Image
   bIsVariable=true
   parent=Overlay_0
   widget=Image_MeterIcon
   properties {
    visibility=SelfHitTestInvisible
    brush {
     bIsDynamicallyLoaded=false
     drawAs=Image
     imageType=NoImage
     mirroring=NoMirror
     resourceName=None
     resourceObject=/Game/Athena/HUD/Materials/M_UI_FuelIcon_Fill
     tiling=NoTile
     imageSize=<27 28>
     margin=<ltrb 0 0 0 0>
     outlineSettings {
      bUseBrushTransparency=false
      roundingType=HalfHeightRadius
      width=0
      color {colorUseRule=UseColor_Specified, specifiedColor=<rgba 0 0 0 0>}
      cornerRadii=<0 0 0 0>
     }
     tintColor {colorUseRule=UseColor_Specified, specifiedColor=<rgba 1 1 1 1>}
     uVRegion {bIsValid=false, max=<0 0>, min=<0 0>}
    }
   }
   slot {class=OverlaySlot, properties {horizontalAlignment=HAlign_Center, verticalAlignment=VAlign_Center}}
  }
  Image_BatteryIcon {
   widgetClassPath=&Image
   bIsVariable=true
   parent=Overlay_0
   widget=Image_BatteryIcon
   properties {
    visibility=Hidden
    brush {
     bIsDynamicallyLoaded=false
     drawAs=Image
     imageType=FullColor
     mirroring=NoMirror
     resourceName=None
     resourceObject=/Game/UI/Foundation/Textures/Icons/Stats/T-Icon-HealthBattery-32
     tiling=NoTile
     imageSize=<44 44>
     margin=<ltrb 0 0 0 0>
     outlineSettings {
      bUseBrushTransparency=false
      roundingType=HalfHeightRadius
      width=0
      color {colorUseRule=UseColor_Specified, specifiedColor=<rgba 0 0 0 0>}
      cornerRadii=<0 0 0 0>
     }
     tintColor {colorUseRule=UseColor_Specified, specifiedColor=<rgba 1 1 1 1>}
     uVRegion {bIsValid=false, max=<0 0>, min=<0 0>}
    }
   }
   slot {class=OverlaySlot, properties {verticalAlignment=VAlign_Center, padding=<ltrb 0 -1 0 0>}}
  }
  Text_FuelInfinite {
   widgetClassPath=&CommonTextBlock
   bIsVariable=true
   parent=VerticalBox_0
   widget=Text_FuelInfinite
   properties {
    text="∞"
    font {
     bForceMonospaced=false
     bMaterialIsStencil=false
     fontMaterial=null
     fontObject=&BurbankSmall
     letterSpacing=0
     monospacedWidth=1
     size=15
     skewAmount=0
     typefaceFontName=Black
     outlineSettings {
      bApplyOutlineToDropShadows=false
      bMiteredCorners=false
      bSeparateFillAlpha=false
      outlineMaterial=null
      outlineSize=0
      outlineColor=<rgba 0 0 0 1>
     }
    }
    renderTransform {angle=0, scale=<1 1>, shear=<-12 0>, translation=<-2 0>}
    shadowOffset=<0 0>
   }
   slot {class=VerticalBoxSlot, properties {horizontalAlignment=HAlign_Center, verticalAlignment=VAlign_Center}}
  }
 ]
}
